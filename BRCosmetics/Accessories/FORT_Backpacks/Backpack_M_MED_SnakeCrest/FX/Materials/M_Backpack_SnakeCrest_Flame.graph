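# Translucent unlit Niagara sprite material that draws a procedural flame for the SnakeCrest backpack, without a shape texture.
# Cropped UVs are pinched by power curves, then distorted by a panning heat-distortion flowmap whose strength grows away from the vertical center.
# MF_Sphere turns those UVs into a normal and mask, driving a Fresnel-like falloff and a two-color gradient scaled by particle alpha.
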Material /BRCosmetics/Accessories/FORT_Backpacks/Backpack_M_MED_SnakeCrest/FX/Materials/M_Backpack_SnakeCrest_Flame
graph {
 properties {bUsedWithNiagaraSprites=true, blendMode=BLEND_Translucent, pixelDepthOffsetMode=PDOM_Legacy, shadingModel=MSM_Unlit, usedShadingModels=Unlit}
 nodes [
  n1 = MaterialFunctionCall(Float2=TextureCoordinate()) {materialFunction=&BreakOutFloat2Components}
  n2 = OneMinus(None=Multiply(A=OneMinus(None=@n1.G), B=ScalarParameter() {defaultValue=0.6, parameterName=Vertical Cropping}))
  n3 = Multiply(
   A=ConstantBiasScale(None=MaterialFunctionCall(X=@n1.R, Y=@n2) {materialFunction=&MakeFloat2}.Result) {bias=-0.5, scale=2}
   B=ScalarParameter() {defaultValue=1.5, parameterName=Cropping}
  )
  ComponentMask(None=Abs(None=@n3)) {g=true}
  Constant()
  n4 = VectorParameter() {parameterName=Color 1, defaultValue=<rgba 0 0.3068 1.66 0.05>}
  n5 = VectorParameter() {parameterName=Color 2, defaultValue=<rgba 0.8 3.2 4 0.133>}
  n6 = ComponentMask(None=Abs(None=@n3)) {g=true}
  n7 = OneMinus(None=@n6)
  n8 = Multiply(A=@n7, B=@n7)
  n9 = ComponentMask(None=@n3) {g=true}
  n10 = OneMinus(None=Power(Base=Abs(None=@n9), Exp=ScalarParameter() {defaultValue=2, parameterName=Exponent 1}))
  n11 = ScalarParameter() {defaultValue=0.45, parameterName=Exponent 2}
  n12 = AppendVector(A=Divide(A=ComponentMask(None=@n3) {r=true}, B=Power(Base=@n10, Exp=@n11)), B=@n9)
  n13 = VectorParameter() {parameterName=Flowmap UV Scale and Speed, defaultValue=<rgba 0.166 0.5 0.2 -1>}
  n14 = Panner(Coordinate=Multiply(A=TextureCoordinate(), B=ComponentMask(None=@n13.RGB) {g=true, r=true}), Speed=AppendVector(A=@n13.B, B=@n13.B)) {bFractionalPart=true}
  n15 = ConstantBiasScale(None=ComponentMask(None=TextureSample(UVs=@n14) {samplerType=SAMPLERTYPE_LinearColor, texture=/BRCosmetics/Packages/Fortress_Desert/Materials/HeatDistortionFlowmap}.RGB) {g=true, r=true}) {bias=-0.5, scale=2}
  n16 = Multiply(
   A=@n6
   B=AppendVector(
    A=ScalarParameter() {defaultValue=3, parameterName=Distortion Intensity X}
    B=ScalarParameter() {defaultValue=2, parameterName=Distortion Intensity Y}
   )
  )
  n17 = MaterialFunctionCall!(UV=Add(A=ConstantBiasScale(None=@n12), B=Multiply(A=@n15, B=@n16)), Size, Curvature) {materialFunction=/Game/Effects/Fort_Effects/MaterialFunctions/MF_Sphere}
  n18 = ComponentMask(None=@n17.Normal) {b=true}
  n19 = Power(Base=OneMinus(None=@n18), Exp=@n4.A)
  n20 = Saturate(None=Divide(A=OneMinus(None=@n19), B=@n5.A))
  Multiply(A=LinearInterpolate(A=@n4.RGB, B=@n5.RGB, Alpha=Multiply(A=@n8, B=@n20)), B=ParticleColor().A)
  Multiply(A=Power(Base=@n18, Exp=ScalarParameter() {defaultValue=1.666, parameterName=Fresnel Exponent}), B=@n17.Mask)
  OneMinus(None=Saturate(None=Multiply(A=@n3) {constB=-1}))
  Saturate()
 ]
}
