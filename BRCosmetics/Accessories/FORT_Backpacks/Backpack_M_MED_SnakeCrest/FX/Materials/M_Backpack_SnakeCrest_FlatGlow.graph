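# Additive unlit Niagara sprite material for a flat glow effect on the SnakeCrest backpack.
# A hard-coded centered SphereMask forms a radial gradient, multiplied by a soft depth fade from the absolute scene-to-pixel depth difference, shaped by falloff distance and exponent parameters.
# The graph also holds a ParticleColor node, a powered mask and a world-Z difference against the position behind translucency, none feeding another node shown.
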Material /BRCosmetics/Accessories/FORT_Backpacks/Backpack_M_MED_SnakeCrest/FX/Materials/M_Backpack_SnakeCrest_FlatGlow
graph {
 properties {bUsedWithNiagaraSprites=true, blendMode=BLEND_Additive, pixelDepthOffsetMode=PDOM_Legacy, shadingModel=MSM_Unlit, usedShadingModels=Unlit}
 nodes [
  ComponentMask(None=Subtract(
    A=WorldPosition().XYZ
    B=MaterialFunctionCall() {materialFunction=&WorldPositionBehindTranslucency}.Result
   )) {b=true}
  n1 = Constant() {r=0.5}
  n2 = SphereMask(A=TextureCoordinate(), B=@n1, Radius=@n1, Hardness=Constant() {r=1})
  n3 = SceneDepth()
  n4 = PixelDepth()
  n5 = Divide(A=Abs(None=Subtract(A=@n3.R, B=@n4.R)), B=ScalarParameter() {defaultValue=3, parameterName=Distance Falloff})
  n6 = ScalarParameter() {defaultValue=1.666, parameterName=Falloff Exponent}
  Multiply(A=@n2, B=Power(Base=Saturate(None=@n5), Exp=@n6))
  ParticleColor()
  Power(Base=@n2, Exp=@n6)
 ]
}
