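# Niagara system for the Snake Crest backpack cosmetic: a persistent cyan flame with glow sprites and a light.
# A single Flame emitter burst-spawns non-expiring particles drawn by three sprite renderers and a light renderer, using custom glow attributes.
# Looping curves driven by system age at different rates flicker size and alpha, with alpha also scaled by time of day.
# Panning curl noise, not written to particle physics, wobbles the glow position.
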NiagaraSystem /BRCosmetics/Accessories/FORT_Backpacks/Backpack_M_MED_SnakeCrest/FX/NS_Backpack_SnakeCrest
dependencies {usedDataInterfaces [&NiagaraDataInterfaceCurve], usedDynamicInputs [], usedModules [&SystemState]}
emitters {Flame {
  emitterUpdateScript [
   &EmitterState(Life Cycle Mode=::NewEnumerator1, Loop Behavior=::NewEnumerator1, Loop Duration Mode=::NewEnumerator1, UseLoopDelay=false)
   &SpawnBurst_Instantaneous(Use Spawn Probability=false, UseLoopCountLimit=false)
  ]
  particleSpawnScript [
   &InitializeParticle(
    Color=<rgba 0 3.5 6 1>
    Position Offset=<0 -2.66 -11>
    Sprite Size=<5 15>
    Sprite Size Mode=::NewEnumerator4
    UsePositionOffset=true
    Write Material Random=true
   )
   &CameraOffset()
   /SetVariables(
    Particles.FlatGlowFacing:Vector3f=<0 -0.4666 1>
    Particles.FlatGlowPosition:NiagaraPosition=<0 -3.5 -9.5>
    Particles.FlatGlowSize:Vector2f=&Vector2DFromFloat(Value=5)
    Particles.GlowCameraOffset=11
    Particles.GlowPosition:NiagaraPosition=<0 -4.5 -7.5>
    Particles.GlowSize=&Vector2DFromFloat(Value=44)
   ) {isSetParameters=true}
   &SpriteFacingAndAlignment(Sprite Alignment=<0 0.55 -1>, WriteAlignment=true, WriteFacing=false)
  ]
  particleUpdateScript [
   &ParticleState(Kill Particles When Lifetime Has Elapsed=false)
   &ScaleSpriteSize(
    Uniform Curve Index=&Frac_Float(Float=&Multiply_Float(A=$Engine.System.Age, B=0.56))
    Uniform Curve Sprite Scale=DataInterface() {propertyValues {Curve {keys [
        {interpMode=RCIM_Cubic, tangentMode=RCTM_Break, time=0, value=1}
        {interpMode=RCIM_Cubic, tangentMode=RCTM_Break, time=0.0437, value=0.9348}
        {interpMode=RCIM_Cubic, tangentMode=RCTM_Break, time=0.1124, value=0.9726}
        {interpMode=RCIM_Cubic, tangentMode=RCTM_Break, time=0.1436, value=0.7838}
        {interpMode=RCIM_Cubic, tangentMode=RCTM_Break, time=0.2212, value=0.8695}
        {interpMode=RCIM_Cubic, tangentMode=RCTM_Break, time=0.2711, value=0.8342}
        {interpMode=RCIM_Cubic, tangentMode=RCTM_Break, time=0.3523, value=0.8995}
        {interpMode=RCIM_Cubic, tangentMode=RCTM_Break, time=0.385, value=0.9488}
        {interpMode=RCIM_Cubic, tangentMode=RCTM_Break, time=0.4551, value=0.8507}
        {interpMode=RCIM_Cubic, tangentMode=RCTM_Break, time=0.5326, value=0.8849}
        {interpMode=RCIM_Cubic, tangentMode=RCTM_Break, time=0.6015, value=0.7808}
        {interpMode=RCIM_Cubic, tangentMode=RCTM_Break, time=0.6326, value=0.8851}
        {interpMode=RCIM_Cubic, tangentMode=RCTM_Break, time=0.7089, value=0.9669}
        {interpMode=RCIM_Cubic, tangentMode=RCTM_Break, time=0.7726, value=0.9248}
        {interpMode=RCIM_Cubic, tangentMode=RCTM_Break, time=0.8751, value=0.908}
        {interpMode=RCIM_Cubic, tangentMode=RCTM_Break, time=1, value=1}
       ]}}}
   )
   &ScaleColor(
    Scale Alpha=&FloatFromCurve(
     FloatCurve=DataInterface() {propertyValues {Curve {keys [
         {interpMode=RCIM_Cubic, tangentMode=RCTM_Break, time=0, value=1}
         {interpMode=RCIM_Cubic, tangentMode=RCTM_Break, time=0.0387, value=0.8877}
         {interpMode=RCIM_Cubic, tangentMode=RCTM_Break, time=0.0888, value=0.9348}
         {interpMode=RCIM_Cubic, tangentMode=RCTM_Break, time=0.1848, value=0.829}
         {interpMode=RCIM_Cubic, tangentMode=RCTM_Break, time=0.3, value=0.85}
         {interpMode=RCIM_Cubic, tangentMode=RCTM_Break, time=0.4, value=0.5666}
         {interpMode=RCIM_Cubic, tangentMode=RCTM_Break, time=0.5, value=0.9}
         {interpMode=RCIM_Cubic, tangentMode=RCTM_Break, time=0.5509, value=0.8482}
         {interpMode=RCIM_Cubic, tangentMode=RCTM_Break, time=0.6409, value=0.9763}
         {interpMode=RCIM_Cubic, tangentMode=RCTM_Break, time=0.6638, value=0.8693}
         {interpMode=RCIM_Cubic, tangentMode=RCTM_Break, time=0.7238, value=0.7751}
         {interpMode=RCIM_Cubic, tangentMode=RCTM_Break, time=0.8932, value=0.83}
         {interpMode=RCIM_Cubic, tangentMode=RCTM_Break, time=0.9145, value=0.9627}
         {interpMode=RCIM_Cubic, tangentMode=RCTM_Break, time=1, value=1}
        ]}}}
     CurveIndex=&Frac_Float(Float=&Multiply_Float(A=$Engine.System.Age, B=0.8))
     Scale Curve=&FNTimeOfDayScalar(Night Scale=0.45, Morning Scale=1, Evening Scale=0.8)
    )
    ScaleA=true
    ScaleRGB=false
   )
   &CurlNoiseForce(
    Noise Frequency=5
    Noise Quality / Cost=::NewEnumerator0
    Noise Strength=1
    Pan Noise=true
    Pan Noise Field=<0 0 3>
    Sample Position:NiagaraPosition=&ConvertVectorToPosition(Input Position=<0 0 0>, Conversion Operation:ENiagara_LWCConvertVecToPos=::NewEnumerator0)
    Write to Intrinsic Parameters=false
   )
   /SetVariables(
    Particles.FlatGlowColor:LinearColor=&MultiplyLinearColor(Linear Color=$Particles.Color, Scale Factor=&MakeVector4_FromFloat(Float=0.4))
    Particles.GlowColor=&MultiplyLinearColor(Linear Color=$Particles.Color, Scale Factor=&MakeVector4_FromFloat(Float=0.166))
    Particles.GlowPositionOffset:NiagaraPosition=&Multiply_VectorByFloat(Vector=$Output.CurlNoiseForce.CurlNoiseForce, Float=1.2)
   ) {isSetParameters=true}
   /SetVariables(Particles.GlowPosition:NiagaraPosition=&AddVectorToPosition(Position=$Particles.Initial.GlowPosition, Vector=Linked() {linkedVariable {name=Particles.GlowPositionOffset, type=NiagaraPosition}})) {isSetParameters=true}
  ]
  rendererClasses [&NiagaraSpriteRendererProperties, &NiagaraLightRendererProperties]
  renderers [
   {rendererClass=&NiagaraSpriteRendererProperties}
   {rendererClass=&NiagaraLightRendererProperties, rendererIndex=1}
   {rendererClass=&NiagaraSpriteRendererProperties, rendererIndex=2}
   {rendererClass=&NiagaraSpriteRendererProperties, rendererIndex=3}
  ]
 }}
summary {
 systemName=NS_Backpack_SnakeCrest
 emitters [{emitterName=Flame, rendererClasses [&NiagaraSpriteRendererProperties, &NiagaraLightRendererProperties]}]
 userVariables [
  User.bIsFrontEnd {type=NiagaraBool}
  User.bIsFrontEndPreview {type=NiagaraBool}
 ]
}
