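# Physics asset for the M_MED_SnakeCrest backpack cosmetic, covering a short bone chain at the part named ballon_top.
# A tiny capsule on base_ballon_top serves as the anchor for a larger, offset and rotated capsule on dyn_ballon_top.
# One constraint joins the two bodies with equal, moderate swing limits and a tighter twist limit, restricting the top piece to a small wobble.
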PhysicsAsset /BRCosmetics/Accessories/FORT_Backpacks/Backpack_M_MED_SnakeCrest/Meshes/M_MED_SnakeCrest_Pack_Physics
bodies {
 base_ballon_top [{shapeType=Capsule, length=0.1, radius=0.5, shapeName=base_ballon_top_capsule}]
 dyn_ballon_top [{shapeType=Capsule, length=3.9701, radius=1.9066, shapeName=dyn_ballon_top_capsule, center=<-2 0 0>, rotation=<rot 90 0 0>}]
}
constraints [{bone1Name=dyn_ballon_top, bone2Name=base_ballon_top, swing1LimitDegrees=20, swing2LimitDegrees=20, twistLimitDegrees=5}]
