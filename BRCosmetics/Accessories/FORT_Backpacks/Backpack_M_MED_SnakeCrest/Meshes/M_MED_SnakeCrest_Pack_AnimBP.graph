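# Animation Blueprint for the SnakeCrest backpack cosmetic, making its balloon float and react to player movement.
# A RigidBody node simulates the pack using per-movement-state gravity overrides, mostly pointing upward for buoyancy.
# Additive bone transforms then offset the root and bob the balloon top with three out-of-phase sine waves.
# Event-graph flags for moving and airborne (jumping, falling, skydiving or parachuting) drive the alphas of extra tilts on those bones.
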AnimBlueprint /BRCosmetics/Accessories/FORT_Backpacks/Backpack_M_MED_SnakeCrest/Meshes/M_MED_SnakeCrest_Pack_AnimBP
target_skeleton=&/M_MED_SnakeCrest_Pack_Skeleton
graphs [
 AnimGraph {nodes [
   n1 = RigidBody!(ComponentPose=Local To Component(LocalPose=Input Pose()), Alpha, OverrideWorldGravity:Vector=Get GravityOverride(), ExternalForce) -> (Pose) {
    bClampLinearTranslationLimitToRefPose=true
    bOverrideWorldGravity=true
    overridePhysicsAsset=&/M_MED_SnakeCrest_Pack_Physics
    simulationSpace=BaseBoneSpace
    alphaBoolBlend {blendOption=Linear}
    alphaScaleBiasClamp {clampMax=0.8}
    baseBoneRef {boneName=root}
    componentAppliedLinearAccClamp=<3000 3000 3000>
    componentLinearAccScale=<0.3 0.4 0.3>
    componentLinearVelScale=<0.3 0.4 0.3>
    overrideWorldGravity {z=-980}
    simSpaceSettings {
     maxAngularAcceleration=500
     maxAngularVelocity=1
     maxLinearAcceleration=20
     maxLinearVelocity=5
     velocityScaleZ=0
     worldAlpha=0.65
     externalLinearDragV=<0.15 0.15>
    }
   }
   n2 = "Transform (Modify) Bone - Bone: base_ballon_top"(ComponentPose=@n1, Alpha=1, Translation=Get balloon_Float() -> (balloon_Float:Vector), Rotation="0,0,0", Scale="1,1,1") {translationMode=BMM_Additive, translationSpace=BCS_BoneSpace, alphaBoolBlend {blendOption=Linear}, boneToModify {boneName=base_ballon_top}}
   n3 = "Transform (Modify) Bone - Bone: root"(ComponentPose=@n2, Translation="0,1.5,0", Rotation="0,0,0") {
    rotationMode=BMM_Additive
    rotationSpace=BCS_BoneSpace
    scaleMode=BMM_Additive
    scaleSpace=BCS_BoneSpace
    translationMode=BMM_Additive
    translationSpace=BCS_BoneSpace
    alphaBoolBlend {blendOption=Linear}
    alphaScaleBiasClamp {bInterpResult=true}
    boneToModify {boneName=root}
    translation {y=1.5}
   }
   n4 = Get Is in Air() -> (Is in Air:double)
   n5 = "Transform (Modify) Bone - Bone: root"(ComponentPose=@n3, Alpha=@n4, Translation="0,0,0", Rotation="0,0,6") {
    rotationMode=BMM_Additive
    rotationSpace=BCS_BoneSpace
    scaleMode=BMM_Additive
    scaleSpace=BCS_BoneSpace
    translationMode=BMM_Additive
    translationSpace=BCS_BoneSpace
    alphaBoolBlend {blendOption=Linear}
    alphaScaleBiasClamp {bInterpResult=true}
    boneToModify {boneName=root}
    rotation {roll=6}
   }
   n6 = "Transform (Modify) Bone - Bone: base_ballon_top"(ComponentPose=@n5, Alpha=@n4, Translation="0,-0.5,0", Rotation="0,-10,0", Scale="1,1,1") {
    rotationMode=BMM_Additive
    rotationSpace=BCS_BoneSpace
    translationMode=BMM_Additive
    translationSpace=BCS_BoneSpace
    alphaBoolBlend {blendOption=Linear}
    alphaScaleBiasClamp {bInterpResult=true}
    boneToModify {boneName=base_ballon_top}
    rotation {yaw=-10}
    translation {y=-0.5}
   }
   n7 = "Transform (Modify) Bone - Bone: base_ballon_top"(ComponentPose=@n6, Alpha=Get is Moving() -> (is Moving:double), Translation="0,0,0", Rotation="0,15,0", Scale="1,1,1") {
    rotationMode=BMM_Additive
    rotationSpace=BCS_BoneSpace
    scaleMode=BMM_Additive
    scaleSpace=BCS_BoneSpace
    translationMode=BMM_Additive
    translationSpace=BCS_BoneSpace
    alphaBoolBlend {blendOption=Linear}
    alphaScaleBiasClamp {bInterpResult=true}
    boneToModify {boneName=base_ballon_top}
    rotation {yaw=15}
   }
   Output Pose(Result=Component To Local(ComponentPose=@n7)) {layerGroup=None}
  ]}
 EventGraph {nodes [
   Event BlueprintUpdateAnimation()
   > n1 = Sequence!(execute) -> (then_0, then_1, then_2:Exec, then_3:Exec)
   Set is Moving(execute=@n1.then_1, is Moving:double="To Float (Boolean)"(InBool=Get bIsMoving2D())) -> (Output_Get:double)
   n2 = Gravity Override State Function DEPRECATED(
    execute=@n1.then_0
    CharacterPartAnimInstance=Self-Reference()
    GameTimeInSeconds=GetGameTimeInSeconds()
    GravityOverrideParams=Get Gravity Override Params()
   )
   > Set GravityOverride(GravityOverride=@n2)
   n3 = OR Boolean!(A=Get bIsSkyDiving(), B=Get bIsSkydiveDiveMode(), C:bool=Get bIsSkydiveFloating(), D:bool=Get bIsParachuteOpened()) -> (ReturnValue)
   n4 = OR Boolean!(A=Get bIsJumping(), B=Get bIsFalling()) -> (ReturnValue)
   Set Is in Air(execute=@n1.then_2, Is in Air:double="To Float (Boolean)"(InBool=OR Boolean(A=@n3, B=@n4))) -> (Output_Get:double)
   n5 = GetRealTimeSeconds()
   n6 = "Sin (Radians)"(A=float * float(A=float * float(A=1.1, B=1.1), B=@n5))
   n7 = "Sin (Radians)"(A=float * float(A=float * float(A=0.9, B=1.1), B=@n5))
   n8 = "Sin (Radians)"(A=float * float(A=float * float(A=1.2, B=1.3), B=@n5))
   Set balloon_Float(
    execute=@n1.then_3
    balloon_Float:Vector=MakeVector(
     X=Remap Float(Input=@n8, OldRange=(X=-1,Y=1), NewRange=(X=-1,Y=0.2)) -> (Output=~)
     Y=float * float(A=@n6, B=0.15)
     Z=float * float(A=@n7, B=0.3)
    )
   ) -> (Output_Get:Vector)
   TryGetPawnOwner()
  ]}
]
variables [
 balloon_Float {type=Vector}
 Gravity Override Params {
  type=GravityOverrideParamsStruct
  default {
   dBNO {gravityOverride=<980 400 0>}
   dBNO_Carried {gravityOverride=<980 500 500>}
   emoteOrMelee {gravityOverride=<0 0 980>}
   front_End {gravityOverride=<0 0 980>}
   onGround_Falling {gravityOverride=<0 0 980>}
   onGround_Moving {gravityOverride=<0 0 980>}
   onGround_Standing {gravityOverride=<0 0 980>}
   rail_Grinding {gravityOverride=<0 0 980>}
   riding_Moving {gravityOverride=<0 450 450>}
   riding_Moving_Targeting {gravityOverride=<0 200 980>}
   riding_Standing {gravityOverride=<0 350 980>}
   skydive_DivingDown {gravityOverride=<1500 0 0>}
   skydive_FloatingDown {gravityOverride=<0 700 600>}
   skydive_GoingUp {gravityOverride=<0 300 1500>}
   skydive_Parachuting {gravityOverride=<0 800 980>}
   swimming {gravityOverride=<800 0 0>}
   tactical_Sprint {gravityOverride=<0 450 500>}
  }
 }
 GravityOverride {type=Vector}
 Is in Air {type=number}
 is Moving {type=number}
 Sim Space Settings {
  type=SimSpaceSettings
  default {
   dampingAlpha=1
   maxAngularAcceleration=500
   maxAngularVelocity=1
   maxLinearAcceleration=20
   maxLinearVelocity=5
   velocityScaleZ=0
   worldAlpha=1
   externalLinearDragV=<0.2 0.2 0>
  }
 }
]
